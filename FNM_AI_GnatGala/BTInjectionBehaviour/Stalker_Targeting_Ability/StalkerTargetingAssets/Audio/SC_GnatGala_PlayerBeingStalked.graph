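# Sound cue for the GnatGala AI's Stalker targeting ability, named as the audio for when a player is being stalked.
# A single wave player feeds a looping node at the root, repeating one borrowed Scary Speaker environment sound indefinitely.
# Its max distance is set extremely high.
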SoundCue /FNM_AI_GnatGala/BTInjectionBehaviour/Stalker_Targeting_Ability/StalkerTargetingAssets/Audio/SC_GnatGala_PlayerBeingStalked
root=Looping_0
nodes [
 {id=Looping_0, children [WavePlayer_0]}
 {id=WavePlayer_0, properties {soundWaveAssetPtr=/Game/Sounds/Helios/Environment_Actors/Scary_Speaker/FNM_Scary_Speaker_Sounds_14}}
]
properties {maxDistance=2097152}
